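# Physics asset for the F_MED_Ephemeral female medium player body cosmetic, covering its dynamic skirt bones plus pelvis, thigh and calf capsules.
# Six-link left and right skirt chains hang from the pelvis with limited swing and twist.
# A three-link center rear chain and two rear side panels attach through tiny root bodies with free angular motion at the pelvis.
# An extra angular-free constraint ties only the left rear panel to the left chain's second link.
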PhysicsAsset /BRCosmetics/Characters/Player/Female/Medium/Bodies/F_MED_Ephemeral/Meshes/F_MED_Ephemeral_Physics
bodies {
 calf_l [{shapeType=Capsule, length=33.8731, radius=5.7365, center=<-22.3753 -2.5909 -0.8217>, rotation=<rot -0.2599 92.0563 -88.8079>}]
 calf_r [{shapeType=Capsule, length=33.8731, radius=5.7365, center=<22.3753 2.5909 0.8217>, rotation=<rot -0.2599 92.0563 91.1921>}]
 dyn_skirt_01_l [{
   shapeType=Capsule
   length=8.1313
   radius=7.9669
   shapeName=dyn_skirt_01_l_capsule
   center=<-7.308 -0.422 1.3067>
   rotation=<rot 4.2128 175.1736 -90>
  }]
 dyn_skirt_01_l_rear [{
   shapeType=Box
   extentX=14.2561
   extentY=14.0458
   extentZ=5.7126
   shapeName=dyn_skirt_01_l_rear_box
   center=<-8.7115 0.9467 -2.2945>
   rotation=<rot -4.6493 -0.1486 -0.0003>
  }]
 dyn_skirt_01_l_rear_root [{shapeType=Capsule, length=0.505, radius=0.505, shapeName=dyn_skirt_01_l_rear_root_capsule, rotation=<rot 0 -90 0>}]
 dyn_skirt_01_r [{
   shapeType=Capsule
   length=8.1313
   radius=7.9669
   shapeName=dyn_skirt_01_l_capsule
   center=<7.308 0.422 -1.3067>
   rotation=<rot 4.2128 175.1736 90>
  }]
 dyn_skirt_01_r_rear [{
   shapeType=Box
   extentX=14.2561
   extentY=14.0458
   extentZ=5.7126
   shapeName=dyn_skirt_01_l_rear_box
   center=<8.7115 -0.9467 2.2945>
   rotation=<rot -4.6493 -0.1486 179.9997>
  }]
 dyn_skirt_01_r_rear_root [{shapeType=Capsule, length=0.505, radius=0.505, shapeName=dyn_skirt_01_r_rear_root_capsule, rotation=<rot 0 -90 0>}]
 dyn_skirt_01_rear [{
   shapeType=Box
   extentX=13.8913
   extentY=18.6085
   extentZ=4.7988
   shapeName=dyn_skirt_01_rear_box
   center=<-7.9927 0 -2.5688>
   rotation=<rot -6.5025 0 0>
  }]
 dyn_skirt_01_rear_root [{shapeType=Capsule, length=0.505, radius=0.505, shapeName=dyn_skirt_01_rear_root_capsule, rotation=<rot 0 -90 0>}]
 dyn_skirt_02_l [{
   shapeType=Capsule
   length=8.3126
   radius=4.3853
   shapeName=dyn_skirt_02_l_capsule
   center=<-6.4107 -0.6933 -0.9471>
   rotation=<rot 0 11.8737 -82.4742>
  }]
 dyn_skirt_02_r [{
   shapeType=Capsule
   length=8.3126
   radius=4.3853
   shapeName=dyn_skirt_02_l_capsule
   center=<6.4107 0.6933 0.9471>
   rotation=<rot 0 11.8737 97.5258>
  }]
 dyn_skirt_02_rear [{shapeType=Capsule, length=8.1955, radius=3.6377, shapeName=dyn_skirt_02_rear_capsule, center=<-5.3492 0.388 -0.1423>, rotation=<rot 0 0 90>}]
 dyn_skirt_03_l [{
   shapeType=Capsule
   length=8.6278
   radius=3.922
   shapeName=dyn_skirt_03_l_capsule
   center=<-2.6737 -1.1211 1.2786>
   rotation=<rot 1.5574 14.316 -80.0645>
  }]
 dyn_skirt_03_r [{
   shapeType=Capsule
   length=8.6278
   radius=3.922
   shapeName=dyn_skirt_03_l_capsule
   center=<2.6737 1.1211 -1.2786>
   rotation=<rot 1.5574 14.316 99.9355>
  }]
 dyn_skirt_03_rear [{shapeType=Capsule, length=4.4194, radius=3.9954, shapeName=dyn_skirt_03_rear_capsule, center=<-4.0535 0.3693 0.271>, rotation=<rot 0 0 90>}]
 dyn_skirt_04_l [{
   shapeType=Capsule
   length=10.1668
   radius=3.0674
   shapeName=dyn_skirt_04_l_capsule
   center=<-2.1379 0.1448 0.8995>
   rotation=<rot 8.5034 9.1279 -85.1207>
  }]
 dyn_skirt_04_r [{
   shapeType=Capsule
   length=10.1668
   radius=3.0674
   shapeName=dyn_skirt_04_l_capsule
   center=<2.1379 -0.1448 -0.8995>
   rotation=<rot 8.5034 9.1279 94.8793>
  }]
 dyn_skirt_05_l [{
   shapeType=Capsule
   length=3.5468
   radius=1.6408
   shapeName=dyn_skirt_05_l_capsule
   center=<-2.7873 0.0857 0.2003>
   rotation=<rot 0 7.0824 -76.8843>
  }]
 dyn_skirt_05_r [{
   shapeType=Capsule
   length=3.5468
   radius=1.6408
   shapeName=dyn_skirt_05_l_capsule
   center=<2.7873 -0.0857 -0.2003>
   rotation=<rot 0 7.0824 103.1157>
  }]
 dyn_skirt_06_l [{
   shapeType=Capsule
   length=2.2208
   radius=0.9126
   shapeName=dyn_skirt_06_l_capsule
   center=<-2.2847 -0.0952 0.0749>
   rotation=<rot 0 91.2747 -88.7818>
  }]
 dyn_skirt_06_r [{
   shapeType=Capsule
   length=2.2208
   radius=0.9126
   shapeName=dyn_skirt_06_l_capsule
   center=<2.2847 0.0952 -0.0749>
   rotation=<rot 0 91.2747 91.2182>
  }]
 pelvis [{shapeType=Capsule, length=10.1834, radius=10.7489, center=<-5.8305 -1.5862 0>}]
 thigh_l [{shapeType=Capsule, length=24.8027, radius=5.9059, center=<-25.3976 0.0071 -0.5329>, rotation=<rot 0.0487 -89.831 -89.2469>}]
 thigh_r [{shapeType=Capsule, length=24.8027, radius=5.9059, center=<25.3976 -0.0071 0.5329>, rotation=<rot 0.0487 -89.831 90.7531>}]
}
constraints [
 {bone1Name=thigh_l, bone2Name=pelvis, swing1LimitDegrees=45, swing2LimitDegrees=45, twistLimitDegrees=45}
 {bone1Name=calf_l, bone2Name=thigh_l, swing1LimitDegrees=45, swing2LimitDegrees=45, twistLimitDegrees=45}
 {bone1Name=dyn_skirt_01_l, bone2Name=pelvis, swing1LimitDegrees=10, swing2LimitDegrees=10, twistLimitDegrees=10}
 {bone1Name=dyn_skirt_02_l, bone2Name=dyn_skirt_01_l, swing1LimitDegrees=12, swing2LimitDegrees=32, twistLimitDegrees=16}
 {bone1Name=dyn_skirt_03_l, bone2Name=dyn_skirt_02_l, swing1LimitDegrees=12, swing2LimitDegrees=32, twistLimitDegrees=16}
 {
  bone1Name=dyn_skirt_01_rear_root
  bone2Name=pelvis
  swing1LimitDegrees=45
  swing1Motion=Free
  swing2LimitDegrees=45
  swing2Motion=Free
  twistLimitDegrees=45
  twistMotion=Free
 }
 {bone1Name=dyn_skirt_01_rear, bone2Name=dyn_skirt_01_rear_root, swing1LimitDegrees=8, swing2LimitDegrees=8, twistLimitDegrees=4}
 {bone1Name=dyn_skirt_02_rear, bone2Name=dyn_skirt_01_rear, swing1LimitDegrees=8, swing2LimitDegrees=8, twistLimitDegrees=4}
 {bone1Name=dyn_skirt_03_rear, bone2Name=dyn_skirt_02_rear, swing1LimitDegrees=8, swing2LimitDegrees=8, twistLimitDegrees=4}
 {bone1Name=calf_r, bone2Name=thigh_r, swing1LimitDegrees=45, swing2LimitDegrees=45, twistLimitDegrees=45}
 {
  bone1Name=dyn_skirt_01_l_rear_root
  bone2Name=pelvis
  swing1LimitDegrees=45
  swing1Motion=Free
  swing2LimitDegrees=45
  swing2Motion=Free
  twistLimitDegrees=45
  twistMotion=Free
 }
 {bone1Name=dyn_skirt_01_l_rear, bone2Name=dyn_skirt_01_l_rear_root, swing1LimitDegrees=8, swing2LimitDegrees=16, twistLimitDegrees=8}
 {
  bone1Name=dyn_skirt_01_r_rear_root
  bone2Name=pelvis
  swing1LimitDegrees=45
  swing1Motion=Free
  swing2LimitDegrees=45
  swing2Motion=Free
  twistLimitDegrees=45
  twistMotion=Free
 }
 {bone1Name=dyn_skirt_01_r_rear, bone2Name=dyn_skirt_01_r_rear_root, swing1LimitDegrees=8, swing2LimitDegrees=16, twistLimitDegrees=8}
 {bone1Name=dyn_skirt_05_r, bone2Name=dyn_skirt_04_r, swing1LimitDegrees=12, swing2LimitDegrees=32, twistLimitDegrees=16}
 {bone1Name=dyn_skirt_06_r, bone2Name=dyn_skirt_05_r, swing1LimitDegrees=12, swing2LimitDegrees=32, twistLimitDegrees=16}
 {
  bone1Name=dyn_skirt_01_l_rear
  bone2Name=dyn_skirt_02_l
  swing1LimitDegrees=45
  swing1Motion=Free
  swing2LimitDegrees=45
  swing2Motion=Free
  twistLimitDegrees=45
  twistMotion=Free
 }
 {bone1Name=dyn_skirt_04_l, bone2Name=dyn_skirt_03_l, swing1LimitDegrees=12, swing2LimitDegrees=32, twistLimitDegrees=16}
 {bone1Name=dyn_skirt_05_l, bone2Name=dyn_skirt_04_l, swing1LimitDegrees=12, swing2LimitDegrees=32, twistLimitDegrees=16}
 {bone1Name=dyn_skirt_06_l, bone2Name=dyn_skirt_05_l, swing1LimitDegrees=12, swing2LimitDegrees=32, twistLimitDegrees=16}
 {bone1Name=dyn_skirt_03_r, bone2Name=dyn_skirt_02_r, swing1LimitDegrees=12, swing2LimitDegrees=32, twistLimitDegrees=16}
 {bone1Name=dyn_skirt_04_r, bone2Name=dyn_skirt_03_r, swing1LimitDegrees=12, swing2LimitDegrees=32, twistLimitDegrees=16}
 {bone1Name=thigh_r, bone2Name=pelvis, swing1LimitDegrees=45, swing2LimitDegrees=45, twistLimitDegrees=45}
 {bone1Name=dyn_skirt_02_r, bone2Name=dyn_skirt_01_r, swing1LimitDegrees=12, swing2LimitDegrees=32, twistLimitDegrees=16}
 {bone1Name=dyn_skirt_01_r, bone2Name=pelvis, swing1LimitDegrees=10, swing2LimitDegrees=10, twistLimitDegrees=10}
]
