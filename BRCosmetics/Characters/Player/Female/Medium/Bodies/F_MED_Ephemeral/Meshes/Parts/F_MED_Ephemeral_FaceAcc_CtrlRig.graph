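# Control Rig for the F_MED_Ephemeral face accessory that swings side and rear hair roots in response to body movement, likely to avoid clipping.
# Setup caches bind and corrective poses for four aim controls by offsetting, capturing and resetting them.
# Each update, spherical pose readers on upper arms, clavicles and head weight sequential pose blends, replaced by parachuting poses when IsParachuting is set.
# Aim constraints point hair roots at those controls, whose nulls follow clavicles and spine_05.
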ControlRigBlueprint /BRCosmetics/Characters/Player/Female/Medium/Bodies/F_MED_Ephemeral/Meshes/Parts/F_MED_Ephemeral_FaceAcc_CtrlRig
graphs [
 DeformRig_GetMirrorTransform_FNC_ContainedGraph {nodes [
   n1 = RigVMFunctionEntryNode()
   n2 = /Script/RigVM.RigVMFunction_MathTransformToEulerTransform(Value=@n1.transform) {is=struct RigVMUnitNode, template="Cast::Execute(in Value,out Result)"}
   > RigVMFunctionReturnNode(
    mirrorTransform=(Rotation=(X=0,Y=0,Z=0,W=1),Translation=(X=0,Y=0,Z=0),Scale3D=(X=1,Y=1,Z=1))
    mirrorTransform.Rotation=/Script/RigVM.RigVMFunction_MathQuaternionFromRotatorV2(Value=@n2.Result.Rotation) {is=struct RigVMUnitNode, template="Cast::Execute(in Value,out Result)"}.Result
    mirrorTransform.Translation=/Script/RigVM.RigVMFunction_MathVectorScale(Factor=-1, Value=@n2.Result.Location) {is=struct RigVMUnitNode, template="Scale::Execute(in Value,in Factor,out Result)"}.Result
    mirrorTransform.Scale3D=@n2.Result.Scale
   )
  ]}
 New Function_ContainedGraph {nodes [
   n1 = RigVMFunctionEntryNode()
   > DISPATCH_RigVMDispatch_ArrayIterator = RigVMDispatchNode(Array=@n1.InControls) {template="DISPATCH_RigVMDispatch_ArrayIterator(in Array,out Element,out Index,out Count,out Ratio)"}
   > &RigUnit_SetTransform(
    Weight=1
    bInitial=False
    Item=@DISPATCH_RigVMDispatch_ArrayIterator.Element
    Value=&RigUnit_GetTransform(bInitial=true, Item=@DISPATCH_RigVMDispatch_ArrayIterator.Element).Transform
   )
   RigVMFunctionReturnNode(ExecuteContext=@DISPATCH_RigVMDispatch_ArrayIterator.Completed)
  ]}
 RigVMFunctionLibrary {nodes [
   mirror_xform = RigVMCollapseNode(ExecuteContext=(), transform=(Rotation=(X=0,Y=0,Z=0,W=1),Translation=(X=0,Y=0,Z=0),Scale3D=(X=1,Y=1,Z=1)))
   reset_controls_to_initial = RigVMCollapseNode(ExecuteContext=(), InControls=())
   set_pose_func = RigVMCollapseNode(ExecuteContext=(), bind_pose=(HierarchyTopologyVersion=-1,PoseHash=-1), target_pose=(HierarchyTopologyVersion=-1,PoseHash=-1))
  ]}
 RigVMModel {nodes [
   RigUnit_BeginExecution = &RigUnit_BeginExecution()
   SequenceAggregate = RigVMAggregateNode(ExecuteContext=@RigUnit_BeginExecution.ExecutePin)
   n1 = &RigUnit_SphericalPoseReader(
    ActiveRegionScaleFactors=(PositiveWidth=3,NegativeWidth=1,PositiveHeight=3,NegativeHeight=1)
    ActiveRegionSize=0.1
    Debug=(bDrawDebug=false,bDraw2D=False,bDrawLocalAxes=False,DebugScale=25,DebugSegments=20,DebugThickness=0.25)
    DriverAxis=(X=1,Y=0,Z=0)
    DriverItem=(Type=Bone,Name="upperarm_l")
    FalloffRegionScaleFactors=(PositiveWidth=1,NegativeWidth=1,PositiveHeight=1,NegativeHeight=1)
    FalloffSize=0.4
    FlipHeightScaling=true
    FlipWidthScaling=true
    OptionalParentItem=(Type=Bone,Name="spine_05")
    RotationOffset=(X=90,Y=-20,Z=0)
    ExecutePin=@SequenceAggregate.A
   )
   > RigVMVariableNode(Value=@n1.OutputParam) {variable=spr_arm_l_up}
   > n2 = &RigUnit_SphericalPoseReader(
    ActiveRegionScaleFactors=(PositiveWidth=3,NegativeWidth=2,PositiveHeight=3,NegativeHeight=2)
    ActiveRegionSize=0.1
    Debug=(bDrawDebug=false,bDraw2D=False,bDrawLocalAxes=False,DebugScale=25,DebugSegments=20,DebugThickness=0.25)
    DriverAxis=(X=1,Y=0,Z=0)
    DriverItem=(Type=Bone,Name="upperarm_l")
    FalloffRegionScaleFactors=(PositiveWidth=1,NegativeWidth=1,PositiveHeight=1,NegativeHeight=1)
    FalloffSize=0.4
    FlipHeightScaling=false
    FlipWidthScaling=true
    OptionalParentItem=(Type=Bone,Name="None")
    RotationOffset=(X=10,Y=0,Z=30)
   )
   > RigVMVariableNode(Value=@n2.OutputParam) {variable=spr_arm_l_extreme_up}
   > n3 = &RigUnit_SphericalPoseReader(
    ActiveRegionScaleFactors=(PositiveWidth=5,NegativeWidth=1,PositiveHeight=1,NegativeHeight=5)
    ActiveRegionSize=0.1
    Debug=(bDrawDebug=false,bDraw2D=False,bDrawLocalAxes=False,DebugScale=25,DebugSegments=20,DebugThickness=0.25)
    DriverAxis=(X=-1,Y=0,Z=0)
    DriverItem=(Type=Bone,Name="upperarm_r")
    FalloffRegionScaleFactors=(PositiveWidth=1,NegativeWidth=1,PositiveHeight=1,NegativeHeight=1)
    FalloffSize=0.5
    FlipHeightScaling=False
    FlipWidthScaling=False
    OptionalParentItem=(Type=Bone,Name="spine_05")
    RotationOffset=(X=-90,Y=20,Z=0)
   )
   > RigVMVariableNode(Value=@n3.OutputParam) {variable=spr_arm_r_up}
   > n4 = &RigUnit_SphericalPoseReader(
    ActiveRegionScaleFactors=(PositiveWidth=3,NegativeWidth=2,PositiveHeight=3,NegativeHeight=2)
    ActiveRegionSize=0.1
    Debug=(bDrawDebug=false,bDraw2D=False,bDrawLocalAxes=False,DebugScale=25,DebugSegments=20,DebugThickness=0.25)
    DriverAxis=(X=-1,Y=0,Z=0)
    DriverItem=(Type=Bone,Name="upperarm_r")
    FalloffRegionScaleFactors=(PositiveWidth=1,NegativeWidth=1,PositiveHeight=1,NegativeHeight=1)
    FalloffSize=0.4
    FlipHeightScaling=False
    FlipWidthScaling=False
    OptionalParentItem=(Type=Bone,Name="None")
    RotationOffset=(X=-170,Y=0,Z=30)
   )
   > RigVMVariableNode(Value=@n4.OutputParam) {variable=spr_arm_r_extreme_up}
   > n5 = &RigUnit_SphericalPoseReader(
    ActiveRegionScaleFactors=(PositiveWidth=1,NegativeWidth=4,PositiveHeight=3,NegativeHeight=3)
    ActiveRegionSize=0.05
    Debug=(bDrawDebug=false,bDraw2D=false,bDrawLocalAxes=false,DebugScale=25,DebugSegments=20,DebugThickness=0.25)
    DriverAxis=(X=-1,Y=0,Z=0)
    DriverItem=(Type=Bone,Name="clavicle_r")
    FalloffRegionScaleFactors=(PositiveWidth=1,NegativeWidth=1,PositiveHeight=1,NegativeHeight=1)
    FalloffSize=0.2
    FlipHeightScaling=False
    FlipWidthScaling=False
    OptionalParentItem=(Type=Bone,Name="None")
    RotationOffset=(X=-10,Y=115,Z=-20)
   )
   > RigVMVariableNode(Value=@n5.OutputParam) {variable=spr_clavicle_r_bk}
   > n6 = &RigUnit_SphericalPoseReader(
    ActiveRegionScaleFactors=(PositiveWidth=1,NegativeWidth=4,PositiveHeight=3,NegativeHeight=3)
    ActiveRegionSize=0.05
    Debug=(bDrawDebug=false,bDraw2D=False,bDrawLocalAxes=False,DebugScale=25,DebugSegments=20,DebugThickness=0.25)
    DriverAxis=(X=1,Y=0,Z=0)
    DriverItem=(Type=Bone,Name="clavicle_l")
    FalloffRegionScaleFactors=(PositiveWidth=1,NegativeWidth=1,PositiveHeight=1,NegativeHeight=1)
    FalloffSize=0.2
    FlipHeightScaling=false
    FlipWidthScaling=false
    OptionalParentItem=(Type=Bone,Name="None")
    RotationOffset=(X=0,Y=-65,Z=-20)
   )
   > RigVMVariableNode(Value=@n6.OutputParam) {variable=spr_clavicle_l_bk}
   > n7 = &RigUnit_SphericalPoseReader(
    ActiveRegionScaleFactors=(PositiveWidth=1,NegativeWidth=1,PositiveHeight=1,NegativeHeight=1)
    ActiveRegionSize=0.1
    Debug=(bDrawDebug=true,bDraw2D=False,bDrawLocalAxes=False,DebugScale=25,DebugSegments=20,DebugThickness=0.25)
    DriverAxis=(X=0,Y=1,Z=0)
    DriverItem=(Type=Bone,Name="head")
    FalloffRegionScaleFactors=(PositiveWidth=1,NegativeWidth=1,PositiveHeight=1,NegativeHeight=1)
    FalloffSize=0.3
    FlipHeightScaling=false
    FlipWidthScaling=false
    OptionalParentItem=(Type=Bone,Name="spine_05")
    RotationOffset=(X=25,Y=180,Z=0)
   )
   > RigVMVariableNode(Value=@n7.OutputParam) {variable=spr_head_look_left}
   &RigUnit_ParentConstraint(
    AdvancedSettings=(InterpolationType=Average,RotationOrderForFilter=XZY)
    Child=(Type=Null,Name="l_hair_side_aimat_ctrl_null")
    Parents=((Item=(Type=Bone,Name="clavicle_l"),Weight=1))
    Weight=1
    ExecutePin=@SequenceAggregate.C
   )
   > &RigUnit_ParentConstraint(
    AdvancedSettings=(InterpolationType=Average,RotationOrderForFilter=XZY)
    Child=(Type=Null,Name="r_hair_side_aimat_ctrl_null")
    Parents=((Item=(Type=Bone,Name="clavicle_r"),Weight=1))
    Weight=1
   )
   > &RigUnit_ParentConstraint(
    AdvancedSettings=(InterpolationType=Average,RotationOrderForFilter=XZY)
    Child=(Type=Null,Name="l_hair_rear_aimat_ctrl_null")
    Parents=((Item=(Type=Bone,Name="spine_05"),Weight=0.5),(Item=(Type=Bone,Name="clavicle_l"),Weight=0.5))
    Weight=1
   )
   > &RigUnit_ParentConstraint(
    AdvancedSettings=(InterpolationType=Average,RotationOrderForFilter=XZY)
    Child=(Type=Null,Name="r_hair_rear_aimat_ctrl_null")
    Parents=((Item=(Type=Bone,Name="spine_05"),Weight=0.5),(Item=(Type=Bone,Name="clavicle_r"),Weight=0.5))
    Weight=1
   )
   RigVMFunction_ControlFlowBranch = &RigVMFunction_ControlFlowBranch(ExecuteContext=@SequenceAggregate.D, Condition=RigVMVariableNode() {variable=IsParachuting}.Value)
   SequenceAggregate_1 = &RigVMFunction_Sequence(ExecuteContext=@RigVMFunction_ControlFlowBranch.False)
   set_pose_func_4 = RigVMFunctionReferenceNode(
    ExecuteContext=@SequenceAggregate_1.A
    bind_pose=RigVMVariableNode() {variable=bind_pose_r_hair_aimat_ctrls}.Value
    target_pose=RigVMVariableNode() {variable=head_look_left_pose}.Value
    pose_alpha=RigVMVariableNode() {variable=spr_head_look_left}.Value
   )
   > set_pose_func = RigVMFunctionReferenceNode(
    bind_pose=/Script/ControlRig.RigUnit_HierarchyGetPoseItemArray(ElementType=All, Initial=false, ItemsToGet=RigVMVariableNode() {variable=r_hair_aimat_ctrls}.Value) {is=struct RigVMUnitNode}.Pose
    target_pose=RigVMVariableNode() {variable=r_arm_raise_lookat_pose}.Value
    pose_alpha=RigVMVariableNode() {variable=spr_arm_r_up}.Value
   )
   > set_pose_func_1 = RigVMFunctionReferenceNode(
    bind_pose=&RigUnit_HierarchyGetPoseItemArray(ElementType=All, Initial=false, ItemsToGet=RigVMVariableNode() {variable=r_hair_aimat_ctrls}.Value).Pose
    target_pose=RigVMVariableNode() {variable=r_arm_raise_extreme_lookat_pose}.Value
    pose_alpha=RigVMVariableNode() {variable=spr_arm_r_extreme_up}.Value
   )
   > set_pose_func_1_3 = RigVMFunctionReferenceNode(
    bind_pose=&RigUnit_HierarchyGetPoseItemArray(ElementType=All, Initial=false, ItemsToGet=RigVMVariableNode() {variable=r_hair_aimat_ctrls}.Value).Pose
    pose_alpha=RigVMVariableNode() {variable=spr_clavicle_r_bk}.Value
    target_pose=RigVMVariableNode() {variable=r_clavicle_bk_lookat_pose}.Value
   )
   set_pose_func_1_1 = RigVMFunctionReferenceNode(
    ExecuteContext=@SequenceAggregate_1.B
    bind_pose=RigVMVariableNode() {variable=bind_pose_l_hair_aimat_ctrls}.Value
    target_pose=RigVMVariableNode() {variable=l_arm_raise_lookat_pose}.Value
    pose_alpha=RigVMVariableNode() {variable=spr_arm_l_up}.Value
   )
   > set_pose_func_1_2 = RigVMFunctionReferenceNode(
    bind_pose=&RigUnit_HierarchyGetPoseItemArray(ElementType=All, Initial=false, ItemsToGet=RigVMVariableNode() {variable=l_hair_aimat_ctrls}.Value).Pose
    pose_alpha=RigVMVariableNode() {variable=spr_arm_l_extreme_up}.Value
    target_pose=RigVMVariableNode() {variable=l_arm_raise_extreme_lookat_pose}.Value
   )
   > set_pose_func_1_2_1 = RigVMFunctionReferenceNode(
    bind_pose=&RigUnit_HierarchyGetPoseItemArray(ElementType=All, Initial=false, ItemsToGet=RigVMVariableNode() {variable=l_hair_aimat_ctrls}.Value).Pose
    pose_alpha=RigVMVariableNode() {variable=spr_clavicle_l_bk}.Value
    target_pose=RigVMVariableNode() {variable=l_clavicle_bk_lookat_pose}.Value
   )
   AlphaInterp = &RigVMFunction_AlphaInterp(InterpSpeedDecreasing=10, InterpSpeedIncreasing=10, Scale=1, Value=1, bClampResult=true, bInterpResult=False)
   set_pose_func_2 = RigVMFunctionReferenceNode(
    pose_alpha=@AlphaInterp.Result
    ExecuteContext=@RigVMFunction_ControlFlowBranch.True
    bind_pose=RigVMVariableNode() {variable=bind_pose_r_hair_aimat_ctrls}.Value
    target_pose=RigVMVariableNode() {variable=Parachuting_Hair_R_Pose}.Value
   )
   > set_pose_func_3 = RigVMFunctionReferenceNode(
    pose_alpha=@AlphaInterp.Result
    bind_pose=RigVMVariableNode() {variable=bind_pose_l_hair_aimat_ctrls}.Value
    target_pose=RigVMVariableNode() {variable=Parachuting_Hair_L_Pose}.Value
   )
   &RigUnit_AimConstraintLocalSpaceOffset(
    Child=(Type=Bone,Name="dyn_side_hair_01_l_root")
    Filter=(bX=True,bY=True,bZ=True)
    Parents=((Item=(Type=Control,Name="l_hair_side_aimat_ctrl"),Weight=1))
    Weight=1
    WorldUp=(Target=(X=0,Y=0,Z=1),Kind=Location,Space=(Type=Bone,Name="clavicle_l"))
    ExecutePin=@SequenceAggregate.E
   )
   > &RigUnit_AimConstraintLocalSpaceOffset(
    Child=(Type=Bone,Name="dyn_side_hair_01_r_root")
    Filter=(bX=True,bY=True,bZ=True)
    Parents=((Item=(Type=Control,Name="r_hair_side_aimat_ctrl"),Weight=1))
    Weight=1
    WorldUp=(Target=(X=0,Y=0,Z=1),Kind=Location,Space=(Type=Bone,Name="clavicle_r"))
   )
   > &RigUnit_AimConstraintLocalSpaceOffset(
    Child=(Type=Bone,Name="dyn_rear_hair_01_l_root")
    Filter=(bX=True,bY=True,bZ=True)
    Parents=((Item=(Type=Control,Name="l_hair_rear_aimat_ctrl"),Weight=1))
    Weight=1
    WorldUp=(Target=(X=0,Y=0,Z=1),Kind=Location,Space=(Type=Bone,Name="spine_05"))
   )
   > &RigUnit_AimConstraintLocalSpaceOffset(
    Child=(Type=Bone,Name="dyn_rear_hair_01_r_root")
    Filter=(bX=True,bY=True,bZ=True)
    Parents=((Item=(Type=Control,Name="r_hair_rear_aimat_ctrl"),Weight=1))
    Weight=1
    WorldUp=(Target=(X=0,Y=0,Z=1),Kind=Location,Space=(Type=Bone,Name="spine_05"))
   )
   RigVMVariableNode() {variable=bind_pose_r_hair_aimat_ctrls}
  ]}
 RigVMModel Rig {nodes [
   n1 = &RigUnit_PrepareForExecution()
   SequenceAggregate = RigVMAggregateNode(ExecuteContext=@n1.ExecutePin)
   RigUnit_ItemArray = &RigUnit_ItemArray(Items=((Type=Control,Name="l_hair_side_aimat_ctrl"),(Type=Control,Name="l_hair_rear_aimat_ctrl")))
   RigVMVariableNode(Value=@RigUnit_ItemArray.Items, ExecuteContext=@SequenceAggregate.A) {variable=l_hair_aimat_ctrls}
   > RigVMVariableNode(Value=&RigUnit_ItemArray(Items=((Type=Control,Name="r_hair_side_aimat_ctrl"),(Type=Control,Name="r_hair_rear_aimat_ctrl"))).Items) {variable=r_hair_aimat_ctrls}
   RigVMVariableNode(
    Value=&RigUnit_HierarchyGetPoseItemArray(ElementType=All, Initial=true, ItemsToGet=RigVMVariableNode() {variable=l_hair_aimat_ctrls}.Value).Pose
    ExecuteContext=@SequenceAggregate.B
   ) {variable=bind_pose_l_hair_aimat_ctrls}
   > RigVMVariableNode(Value=&RigUnit_HierarchyGetPoseItemArray(ElementType=All, Initial=true, ItemsToGet=RigVMVariableNode() {variable=r_hair_aimat_ctrls}.Value).Pose) {variable=bind_pose_r_hair_aimat_ctrls}
   SequenceAggregate_1 = RigVMAggregateNode(ExecuteContext=@SequenceAggregate.C)
   &RigUnit_ModifyTransforms(
    ItemToModify="((Item=(Type=Control,Name=\"r_hair_rear_aimat_ctrl\"),Transform=(Rotation=(X=0,Y=0,Z=0,W=1),Translation=(X=0,Y=0,Z=0),Scale3D=(X=1,Y=1,Z=1))),(Item=(Type=Control,Name=\"r_hair_side_aimat_ctrl\"),Transform=(Rotation=(X=0,Y=0,Z=0,W=1),Translation=(X=0,Y=-2,Z=0), ...(truncated)"
    Weight=1
    WeightMinimum=0
    ExecutePin=@SequenceAggregate_1.A
   )
   > RigVMVariableNode(Value=&RigUnit_HierarchyGetPoseItemArray(ElementType=All, Initial=false, ItemsToGet=RigVMVariableNode() {variable=r_hair_aimat_ctrls}.Value).Pose) {variable=head_look_left_pose}
   > &RigUnit_ModifyTransforms(
    ItemToModify="((Item=(Type=Control,Name=\"r_hair_rear_aimat_ctrl\"),Transform=(Rotation=(X=0,Y=0,Z=0,W=1),Translation=(X=4,Y=-10,Z=2),Scale3D=(X=1,Y=1,Z=1))),(Item=(Type=Control,Name=\"r_hair_side_aimat_ctrl\"),Transform=(Rotation=(X=0,Y=0,Z=0,W=1),Translation=(X=-2,Y=5,Z=0 ...(truncated)"
    Weight=1
    WeightMinimum=0
   )
   > RigVMVariableNode(Value=&RigUnit_HierarchyGetPoseItemArray(ElementType=All, Initial=false, ItemsToGet=RigVMVariableNode() {variable=r_hair_aimat_ctrls}.Value).Pose) {variable=r_arm_raise_lookat_pose}
   > &RigUnit_ModifyTransforms(
    ItemToModify="((Item=(Type=Control,Name=\"r_hair_rear_aimat_ctrl\"),Transform=(Rotation=(X=0,Y=0,Z=0,W=1),Translation=(X=0,Y=3,Z=0),Scale3D=(X=1,Y=1,Z=1))),(Item=(Type=Control,Name=\"r_hair_side_aimat_ctrl\"),Transform=(Rotation=(X=0,Y=0,Z=0,W=1),Translation=(X=0,Y=3,Z=-3), ...(truncated)"
    Weight=1
    WeightMinimum=0
   )
   > RigVMVariableNode(Value=&RigUnit_HierarchyGetPoseItemArray(ElementType=All, Initial=false, ItemsToGet=RigVMVariableNode() {variable=r_hair_aimat_ctrls}.Value).Pose) {variable=r_arm_raise_extreme_lookat_pose}
   > &RigUnit_ModifyTransforms(
    ItemToModify="((Item=(Type=Control,Name=\"r_hair_rear_aimat_ctrl\"),Transform=(Rotation=(X=0,Y=0,Z=0,W=1),Translation=(X=0,Y=-2.55,Z=0),Scale3D=(X=1,Y=1,Z=1))),(Item=(Type=Control,Name=\"r_hair_side_aimat_ctrl\"),Transform=(Rotation=(X=0,Y=0,Z=0,W=1),Translation=(X=-3,Y=3,Z ...(truncated)"
    Weight=1
    WeightMinimum=0
   )
   > RigVMVariableNode(Value=&RigUnit_HierarchyGetPoseItemArray(ElementType=All, Initial=false, ItemsToGet=RigVMVariableNode() {variable=r_hair_aimat_ctrls}.Value).Pose) {variable=r_clavicle_bk_lookat_pose}
   > reset_controls_to_initial_1 = RigVMFunctionReferenceNode(InControls=RigVMVariableNode() {variable=r_hair_aimat_ctrls}.Value)
   &RigUnit_ModifyTransforms(
    ItemToModify="((Item=(Type=Control,Name=\"l_hair_rear_aimat_ctrl\"),Transform=(Rotation=(X=0,Y=0,Z=0,W=1),Translation=(X=-4,Y=0,Z=-2),Scale3D=(X=1,Y=1,Z=1))),(Item=(Type=Control,Name=\"l_hair_side_aimat_ctrl\"),Transform=(Rotation=(X=0,Y=0,Z=0,W=1),Translation=(X=2,Y=-3,Z=3 ...(truncated)"
    Weight=1
    WeightMinimum=0
    ExecutePin=@SequenceAggregate_1.B
   )
   > RigVMVariableNode(Value=&RigUnit_HierarchyGetPoseItemArray(ElementType=All, Initial=false, ItemsToGet=RigVMVariableNode() {variable=l_hair_aimat_ctrls}.Value).Pose) {variable=l_arm_raise_lookat_pose}
   > &RigUnit_ModifyTransforms(
    ItemToModify="((Item=(Type=Control,Name=\"l_hair_rear_aimat_ctrl\"),Transform=(Rotation=(X=0,Y=0,Z=0,W=1),Translation=(X=0,Y=-2.55,Z=0),Scale3D=(X=1,Y=1,Z=1))),(Item=(Type=Control,Name=\"l_hair_side_aimat_ctrl\"),Transform=(Rotation=(X=0,Y=0,Z=0,W=1),Translation=(X=2,Y=-2.5 ...(truncated)"
    Weight=1
    WeightMinimum=0
   )
   > RigVMVariableNode(Value=&RigUnit_HierarchyGetPoseItemArray(ElementType=All, Initial=false, ItemsToGet=RigVMVariableNode() {variable=l_hair_aimat_ctrls}.Value).Pose) {variable=l_arm_raise_extreme_lookat_pose}
   > reset_controls_to_initial = RigVMFunctionReferenceNode(InControls=RigVMVariableNode() {variable=l_hair_aimat_ctrls}.Value)
   > &RigUnit_ModifyTransforms(
    ItemToModify="((Item=(Type=Control,Name=\"l_hair_rear_aimat_ctrl\"),Transform=(Rotation=(X=0,Y=0,Z=0,W=1),Translation=(X=0,Y=-2.55,Z=0),Scale3D=(X=1,Y=1,Z=1))),(Item=(Type=Control,Name=\"l_hair_side_aimat_ctrl\"),Transform=(Rotation=(X=0,Y=0,Z=0,W=1),Translation=(X=3,Y=-3,Z ...(truncated)"
    Weight=1
    WeightMinimum=0
   )
   > RigVMVariableNode(Value=&RigUnit_HierarchyGetPoseItemArray(ElementType=All, Initial=false, ItemsToGet=RigVMVariableNode() {variable=l_hair_aimat_ctrls}.Value).Pose) {variable=l_clavicle_bk_lookat_pose}
   > reset_controls_to_initial_2 = RigVMFunctionReferenceNode(InControls=RigVMVariableNode() {variable=l_hair_aimat_ctrls}.Value)
   &RigUnit_ModifyTransforms(
    ItemToModify="((Item=(Type=Control,Name=\"r_hair_rear_aimat_ctrl\"),Transform=(Rotation=(X=0,Y=0,Z=0,W=1),Translation=(X=0,Y=0,Z=2),Scale3D=(X=1,Y=1,Z=1))),(Item=(Type=Control,Name=\"r_hair_side_aimat_ctrl\"),Transform=(Rotation=(X=0,Y=0,Z=0,W=1),Translation=(X=-5,Y=8,Z=-2) ...(truncated)"
    Weight=1
    WeightMinimum=0
    ExecutePin=@SequenceAggregate_1.C
   )
   > RigVMVariableNode(Value=&RigUnit_HierarchyGetPoseItemArray(ElementType=All, Initial=false, ItemsToGet=RigVMVariableNode() {variable=r_hair_aimat_ctrls}.Value).Pose) {variable=Parachuting_Hair_R_Pose}
   > &RigUnit_ModifyTransforms(
    ItemToModify="((Item=(Type=Control,Name=\"l_hair_rear_aimat_ctrl\"),Transform=(Rotation=(X=0,Y=0,Z=0,W=1),Translation=(X=0,Y=0,Z=-2),Scale3D=(X=1,Y=1,Z=1))),(Item=(Type=Control,Name=\"l_hair_side_aimat_ctrl\"),Transform=(Rotation=(X=0,Y=0,Z=0,W=1),Translation=(X=5,Y=-8,Z=2) ...(truncated)"
    Weight=1
    WeightMinimum=0
   )
   > RigVMVariableNode(Value=&RigUnit_HierarchyGetPoseItemArray(ElementType=All, Initial=false, ItemsToGet=RigVMVariableNode() {variable=l_hair_aimat_ctrls}.Value).Pose) {variable=Parachuting_Hair_L_Pose}
   > reset_controls_to_initial_1_2 = RigVMFunctionReferenceNode(InControls=RigVMVariableNode() {variable=r_hair_aimat_ctrls}.Value)
   > reset_controls_to_initial_2_1 = RigVMFunctionReferenceNode(InControls=RigVMVariableNode() {variable=l_hair_aimat_ctrls}.Value)
  ]}
 SequenceAggregate_1_1_ContainedGraph {nodes [
   n1 = RigVMFunctionEntryNode()
   SequenceAggregate_1 = &RigVMFunction_Sequence(ExecuteContext=@n1.ExecuteContext)
   SequenceAggregate_1_1 = &RigVMFunction_Sequence(ExecuteContext=@SequenceAggregate_1.B)
   RigVMFunctionReturnNode(A=@SequenceAggregate_1.A, B=@SequenceAggregate_1_1.A, C=@SequenceAggregate_1_1.B)
  ]}
 SequenceAggregate_1_ContainedGraph {nodes [
   n1 = RigVMFunctionEntryNode()
   SequenceAggregate = &RigVMFunction_Sequence(ExecuteContext=@n1.ExecuteContext)
   SequenceAggregate_2 = &RigVMFunction_Sequence(ExecuteContext=@SequenceAggregate.B)
   SequenceAggregate_3 = &RigVMFunction_Sequence(ExecuteContext=@SequenceAggregate_2.B)
   RigVMFunctionReturnNode(A=@SequenceAggregate.A, C=@SequenceAggregate_2.A, D=@SequenceAggregate_3.A, E=@SequenceAggregate_3.B)
  ]}
 SequenceAggregate_1_ContainedGraph {nodes [
   n1 = RigVMFunctionEntryNode()
   SequenceAggregate = &RigVMFunction_Sequence(ExecuteContext=@n1.ExecuteContext)
   SequenceAggregate_1 = &RigVMFunction_Sequence(ExecuteContext=@SequenceAggregate.B)
   RigVMFunctionReturnNode(A=@SequenceAggregate.A, B=@SequenceAggregate_1.A, C=@SequenceAggregate_1.B)
  ]}
 set_pose_func_ContainedGraph {nodes [
   n1 = RigVMFunctionEntryNode()
   SequenceAggregate = &RigVMFunction_Sequence(ExecuteContext=@n1.ExecuteContext)
   Subtract = &RigVMFunction_MathDoubleSub(A=1, B=@n1.pose_alpha)
   /Script/ControlRig.RigUnit_HierarchySetPoseItemArray(ElementType=All, ItemsToSet=(), Pose=@n1.bind_pose, ExecutePin=@SequenceAggregate.A, Weight=@Subtract.Result) {is=struct RigVMUnitNode}
   &RigUnit_HierarchySetPoseItemArray(ElementType=All, ItemsToSet=(), Weight=@n1.pose_alpha, Pose=@n1.target_pose, ExecutePin=@SequenceAggregate.B)
   > RigVMFunctionReturnNode()
  ]}
]
hierarchy [
 root {loc=<0 0 0>, rot=<0 0 0 1>}
 pelvis {parent=root, loc=<0 -0.5301 98.0571>, rot=<0.0127 0.707 -0.0127 -0.707>}
 spine_01 {parent=pelvis, loc=<6.1208 -0.9328 0>, rot=<0 0 0.034 -0.9994>}
 spine_02 {parent=spine_01, loc=<6.5996 -0.2095 0>, rot=<0 0 0.0238 -0.9997>}
 spine_03 {parent=spine_02, loc=<6.4914 -0.1237 0>, rot=<0 0 0.0149 -0.9999>}
 spine_04 {parent=spine_03, loc=<6.9598 0.0794 0>, rot=<0 0 -0.0132 -0.9999>}
 spine_05 {parent=spine_04, loc=<6.9987 0 0>, rot=<0 0 -0.0015 -1>}
 neck_01 {parent=spine_05, loc=<11.3138 0.1529 0>, rot=<0 0 -0.2003 -0.9797>}
 neck_02 {parent=neck_01, loc=<4.5627 0 0>, rot=<0 0 0.0504 -0.9987>}
 head {parent=neck_02, loc=<5.3914 0 0>, rot=<0 0 0.115 -0.9934>}
 dyn_front_hair_01_l_root {parent=head, loc=<10.083 5.345 -4.3183>, rot=<-0.0606 0.1086 -0.0932 -0.9878>}
 dyn_front_hair_01_l {parent=dyn_front_hair_01_l_root, loc=<0 0 0>, rot=<0 0 0 1>}
 dyn_front_hair_02_l {parent=dyn_front_hair_01_l, loc=<-7.5 0 0>, rot=<0.0117 -0.1189 0.056 -0.9913>}
 dyn_front_hair_03_l {parent=dyn_front_hair_02_l, loc=<-7.5 0 0>, rot=<-0.0004 0.0035 0.1054 -0.9944>}
 dyn_front_hair_04_l {parent=dyn_front_hair_03_l, loc=<-7.5 0 0>, rot=<0 0.008 0 -1>}
 dyn_front_hair_01_r_root {parent=head, loc=<10.0829 5.345 4.3183>, rot=<0.1086 0.0606 0.9878 -0.0932>}
 dyn_front_hair_01_r {parent=dyn_front_hair_01_r_root, loc=<0 0 0>, rot=<0 0 0 1>}
 dyn_front_hair_02_r {parent=dyn_front_hair_01_r, loc=<7.4998 0 0>, rot=<0.0117 -0.1189 0.056 -0.9913>}
 dyn_front_hair_03_r {parent=dyn_front_hair_02_r, loc=<7.4998 0 0>, rot=<-0.0004 0.0035 0.1054 -0.9944>}
 dyn_front_hair_04_r {parent=dyn_front_hair_03_r, loc=<7.5003 0 0>, rot=<0 0.008 0 -1>}
 dyn_side_hair_01_l_root {parent=head, loc=<0.2703 -1.6502 -6.8606>, rot=<-0.0386 0.0614 -0.0933 -0.993>}
 dyn_side_hair_01_l {parent=dyn_side_hair_01_l_root, loc=<0 0 0>, rot=<0 0 0 1>}
 dyn_side_hair_02_l {parent=dyn_side_hair_01_l, loc=<-5.3 0 0>, rot=<0.1781 0.1664 -0.0543 -0.9683>}
 dyn_side_hair_03_l {parent=dyn_side_hair_02_l, loc=<-6.2712 0 0>, rot=<0.04 0.3666 -0.0319 -0.929>}
 dyn_side_hair_04_l {parent=dyn_side_hair_03_l, loc=<-5.6129 0 0>, rot=<0 0.008 0 -1>}
 dyn_side_hair_01_r_root {parent=head, loc=<0.2704 -1.6502 6.8607>, rot=<0.0614 0.0386 0.993 -0.0933>}
 dyn_side_hair_01_r {parent=dyn_side_hair_01_r_root, loc=<0 0 0>, rot=<0 0 0 1>}
 dyn_side_hair_02_r {parent=dyn_side_hair_01_r, loc=<5.3005 -0.0001 -0.0001>, rot=<0.1781 0.1664 -0.0543 -0.9683>}
 dyn_side_hair_03_r {parent=dyn_side_hair_02_r, loc=<6.2704 0.0002 0.0004>, rot=<0.04 0.3666 -0.0319 -0.929>}
 dyn_side_hair_04_r {parent=dyn_side_hair_03_r, loc=<5.6131 -0.0001 -0.0005>, rot=<0 0.008 0 -1>}
 dyn_rear_hair_01_l_root {parent=head, loc=<5.1234 -6.9029 -4.0064>, rot=<0.5114 0.0121 0.0131 -0.8592>}
 dyn_rear_hair_01_l {parent=dyn_rear_hair_01_l_root, loc=<0 0 0>, rot=<0 0 0 1>}
 dyn_rear_hair_02_l {parent=dyn_rear_hair_01_l, loc=<-5.3 0 0>, rot=<0.0466 0.0588 0.0108 -0.9971>}
 dyn_rear_hair_03_l {parent=dyn_rear_hair_02_l, loc=<-6.2712 0 0>, rot=<-0.0403 0.3716 0.0255 -0.9272>}
 dyn_rear_hairstrand_a_01_l {parent=dyn_rear_hair_03_l, loc=<-2.5424 1.6728 1.1811>, rot=<-0.0204 -0.0739 0.0871 -0.9932>}
 dyn_rear_hairstrand_a_02_l {parent=dyn_rear_hairstrand_a_01_l, loc=<-6.7169 0 0>, rot=<-0.0067 0.1542 0.0665 -0.9858>}
 dyn_rear_hairstrand_a_03_l {parent=dyn_rear_hairstrand_a_02_l, loc=<-5.2687 0 0>, rot=<0 0 0 1>}
 dyn_rear_hairstrand_b_01_l {parent=dyn_rear_hair_03_l, loc=<-2.8103 -1.504 1.4927>, rot=<-0.0394 -0.0808 0.0257 -0.9956>}
 dyn_rear_hairstrand_b_02_l {parent=dyn_rear_hairstrand_b_01_l, loc=<-5.5555 0 0>, rot=<0 0.1755 0 -0.9845>}
 dyn_rear_hairstrand_c_01_l {parent=dyn_rear_hair_03_l, loc=<-2.3071 -3.5187 2.1131>, rot=<0.0433 -0.0556 -0.0747 -0.9947>}
 dyn_rear_hairstrand_c_02_l {parent=dyn_rear_hairstrand_c_01_l, loc=<-3.9258 0 0>, rot=<0 0 0 1>}
 dyn_rear_hair_01_r_root {parent=head, loc=<5.123 -6.9029 4.0064>, rot=<0.0121 -0.5114 0.8592 0.0131>}
 dyn_rear_hair_01_r {parent=dyn_rear_hair_01_r_root, loc=<0 0 0>, rot=<0 0 0 -1>}
 dyn_rear_hair_02_r {parent=dyn_rear_hair_01_r, loc=<5.3 0 0>, rot=<0.0466 0.0588 0.0108 -0.9971>}
 dyn_rear_hair_03_r {parent=dyn_rear_hair_02_r, loc=<6.2705 0 0.0001>, rot=<-0.0403 0.3716 0.0255 -0.9272>}
 dyn_rear_hairstrand_a_01_r {parent=dyn_rear_hair_03_r, loc=<2.5424 -1.6728 -1.1812>, rot=<-0.0204 -0.0739 0.0871 -0.9932>}
 dyn_rear_hairstrand_a_02_r {parent=dyn_rear_hairstrand_a_01_r, loc=<6.7174 0.0002 -0.0005>, rot=<-0.0067 0.1542 0.0665 -0.9858>}
 dyn_rear_hairstrand_a_03_r {parent=dyn_rear_hairstrand_a_02_r, loc=<5.2685 -0.0002 0.0004>, rot=<0 0 0 1>}
 dyn_rear_hairstrand_b_01_r {parent=dyn_rear_hair_03_r, loc=<2.8107 1.504 -1.4932>, rot=<-0.0394 -0.0808 0.0257 -0.9956>}
 dyn_rear_hairstrand_b_02_r {parent=dyn_rear_hairstrand_b_01_r, loc=<5.5554 0 0.0002>, rot=<0 0.1755 0 -0.9845>}
 dyn_rear_hairstrand_c_01_r {parent=dyn_rear_hair_03_r, loc=<2.3072 3.5187 -2.1131>, rot=<0.0433 -0.0556 -0.0747 -0.9947>}
 dyn_rear_hairstrand_c_02_r {parent=dyn_rear_hairstrand_c_01_r, loc=<3.9261 0 -0.0002>, rot=<0 0 0 1>}
 clavicle_l {parent=spine_05, loc=<5.4831 -0.1053 -3.2946>, rot=<0.0465 -0.7289 -0.009 -0.6829>}
 upperarm_l {parent=clavicle_l, loc=<9.8658 -1.2136 -0.4105>, rot=<0.0164 -0.3245 -0.0269 -0.9454>}
 clavicle_r {parent=spine_05, loc=<5.4831 -0.1053 3.2946>, rot=<0.7289 0.0465 -0.6829 0.009>}
 upperarm_r {parent=clavicle_r, loc=<-9.8658 1.2136 0.4105>, rot=<0.0164 -0.3245 -0.0269 -0.9454>}
 l_hair_side_aimat_ctrl_null {type=Null, loc=<13.0371 -5.8204 135.5446>, rot=<-0.0065 0.3563 0.0039 0.9343>}
 l_hair_side_aimat_ctrl {
  controlType=EULER_TRANSFORM
  parent=l_hair_side_aimat_ctrl_null
  type=Control
  loc=<80.5512 7.2562 -109.7043>
  rot=<0.0065 -0.3563 -0.0039 0.9343>
 }
 r_hair_side_aimat_ctrl_null {type=Null, loc=<-13.0371 -5.8204 135.5446>, rot=<-0.9343 0.0039 -0.3563 -0.0065>}
 r_hair_side_aimat_ctrl {
  controlType=EULER_TRANSFORM
  parent=r_hair_side_aimat_ctrl_null
  type=Control
  loc=<-80.5512 -7.2562 109.7043>
  rot=<0.9343 -0.0039 0.3563 -0.0065>
 }
 l_hair_rear_aimat_ctrl_null {type=Null, loc=<6.793 -9.0354 129.217>, rot=<0.0284 -0.7065 -0.0284 0.7065>}
 l_hair_rear_aimat_ctrl {
  controlType=EULER_TRANSFORM
  parent=l_hair_rear_aimat_ctrl_null
  type=Control
  loc=<-129.5255 -1.3485 6.793>
  rot=<-0.0284 0.7065 0.0284 0.7065>
 }
 r_hair_rear_aimat_ctrl_null {type=Null, loc=<-6.793 -9.0354 129.217>, rot=<0.0284 -0.7065 -0.0284 0.7065>}
 r_hair_rear_aimat_ctrl {
  controlType=EULER_TRANSFORM
  parent=r_hair_rear_aimat_ctrl_null
  type=Control
  loc=<-129.5255 -1.3485 -6.793>
  rot=<-0.0284 0.7065 0.0284 0.7065>
 }
]
variables [
 IsParachuting {type=bool, default=False}
 Parachuting_Hair_L_Pose {type=FRigPose, default=(Elements=,HierarchyTopologyVersion=-1,PoseHash=-1)}
 Parachuting_Hair_R_Pose {type=FRigPose, default=(Elements=,HierarchyTopologyVersion=-1,PoseHash=-1)}
 bind_pose_l_hair_aimat_ctrls {type=FRigPose, default=(Elements=,HierarchyTopologyVersion=-1,PoseHash=-1)}
 bind_pose_r_hair_aimat_ctrls {type=FRigPose, default=(Elements=,HierarchyTopologyVersion=-1,PoseHash=-1)}
 head_look_left_pose {type=FRigPose, default=(Elements=,HierarchyTopologyVersion=-1,PoseHash=-1)}
 l_arm_raise_extreme_lookat_pose {type=FRigPose, default=(Elements=,HierarchyTopologyVersion=-1,PoseHash=-1)}
 l_arm_raise_lookat_pose {type=FRigPose, default=(Elements=,HierarchyTopologyVersion=-1,PoseHash=-1)}
 l_clavicle_bk_lookat_pose {type=FRigPose, default=(Elements=,HierarchyTopologyVersion=-1,PoseHash=-1)}
 l_hair_aimat_ctrls {type="TArray<FRigElementKey>", default=((Type=Control,Name="l_hair_side_aimat_ctrl"),(Type=Control,Name="l_hair_rear_aimat_ctrl"))}
 r_arm_raise_extreme_lookat_pose {type=FRigPose, default=(Elements=,HierarchyTopologyVersion=-1,PoseHash=-1)}
 r_arm_raise_lookat_pose {type=FRigPose, default=(Elements=,HierarchyTopologyVersion=-1,PoseHash=-1)}
 r_clavicle_bk_lookat_pose {type=FRigPose, default=(Elements=,HierarchyTopologyVersion=-1,PoseHash=-1)}
 r_hair_aimat_ctrls {type="TArray<FRigElementKey>", default=((Type=Control,Name="r_hair_side_aimat_ctrl"),(Type=Control,Name="r_hair_rear_aimat_ctrl"))}
 spr_arm_l_extreme_up {type=float, default=0}
 spr_arm_l_up {type=float, default=0}
 spr_arm_r_extreme_up {type=float, default=0}
 spr_arm_r_up {type=float, default=0}
 spr_clavicle_l_bk {type=float, default=0}
 spr_clavicle_r_bk {type=float, default=0}
 spr_head_bk {type=float, default=0}
 spr_head_lean_l_bk {type=float, default=0}
 spr_head_lean_r_bk {type=float, default=0}
 spr_head_look_left {type=float, default=0}
]
