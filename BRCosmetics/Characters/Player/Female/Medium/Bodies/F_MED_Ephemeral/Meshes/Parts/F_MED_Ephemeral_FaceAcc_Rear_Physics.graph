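# Physics asset for the Ephemeral medium female character's rear face accessory, defining capsule bodies and joints for dynamic hair.
# Head, neck, upper spine, clavicle and upper-arm capsules form the body the hair can collide with.
# Mirrored front, side and rear hair chains hang from the head, each rear chain forking into three strands.
# Front chains loosen toward the tips; side and rear joints stay tight, with some twist limits differing between left and right.
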PhysicsAsset /BRCosmetics/Characters/Player/Female/Medium/Bodies/F_MED_Ephemeral/Meshes/Parts/F_MED_Ephemeral_FaceAcc_Rear_Physics
bodies {
 clavicle_l [{
   shapeType=Capsule
   length=11.0711
   radius=5.7159
   shapeName=clavicle_l_capsule
   center=<3.9979 -1.4804 -1.0182>
   rotation=<rot -0.103 -88.4202 93.7211>
  }]
 clavicle_r [{
   shapeType=Capsule
   length=11.0711
   radius=5.7159
   shapeName=clavicle_l_capsule
   center=<-3.9979 1.4804 1.0182>
   rotation=<rot -0.103 -88.4202 -86.2789>
  }]
 dyn_front_hair_01_l [{
   shapeType=Capsule
   length=3.6838
   radius=2.2008
   shapeName=dyn_front_hair_01_l_capsule
   center=<-5.092 -0.0345 -1.6704>
   rotation=<rot 0.2358 -91.2303 -77.408>
  }]
 dyn_front_hair_01_l_root [{shapeType=Capsule, length=0.505, radius=0.505, shapeName=dyn_front_hair_01_l_root_capsule, rotation=<rot 0 -90 0>}]
 dyn_front_hair_01_r [{
   shapeType=Capsule
   length=3.6838
   radius=2.2008
   shapeName=dyn_front_hair_01_l_capsule
   center=<5.092 0.0345 1.6704>
   rotation=<rot 0.2358 -91.2303 102.592>
  }]
 dyn_front_hair_01_r_root [{shapeType=Capsule, length=0.505, radius=0.505, shapeName=dyn_front_hair_01_r_root_capsule, rotation=<rot 0 -90 0>}]
 dyn_front_hair_02_l [{
   shapeType=Capsule
   length=4.3284
   radius=2.0796
   shapeName=dyn_front_hair_02_l_capsule
   center=<-4.242 -0.2025 -1.2401>
   rotation=<rot -4.4265 -87.2086 -84.4917>
  }]
 dyn_front_hair_02_r [{
   shapeType=Capsule
   length=4.3284
   radius=2.0796
   shapeName=dyn_front_hair_02_l_capsule
   center=<4.242 0.2025 1.2401>
   rotation=<rot -4.4265 -87.2086 95.5083>
  }]
 dyn_front_hair_03_l [{
   shapeType=Capsule
   length=5.2671
   radius=1.8503
   shapeName=dyn_front_hair_03_l_capsule
   center=<-3.0426 -0.2375 -0.0066>
   rotation=<rot -89.8762 -0.0092 -0.0005>
  }]
 dyn_front_hair_03_r [{
   shapeType=Capsule
   length=5.2671
   radius=1.8503
   shapeName=dyn_front_hair_03_l_capsule
   center=<3.0426 0.2375 0.0066>
   rotation=<rot -89.8762 -0.0092 179.9995>
  }]
 dyn_front_hair_04_l [{
   shapeType=Capsule
   length=5.4164
   radius=1.6403
   shapeName=dyn_front_hair_04_l_capsule
   center=<-3.2765 1.0605 0.0301>
   rotation=<rot -89.4731 179.9982 164.0417>
  }]
 dyn_front_hair_04_r [{
   shapeType=Capsule
   length=5.4164
   radius=1.6403
   shapeName=dyn_front_hair_04_l_capsule
   center=<3.2765 -1.0605 -0.0301>
   rotation=<rot -89.4731 179.9982 -15.9583>
  }]
 dyn_rear_hair_01_l [{shapeType=Capsule, length=2, radius=2, shapeName=dyn_rear_hair_01_l_capsule, center=<-3.9413 -0.05 0.1689>, rotation=<rot 0 90 90>}]
 dyn_rear_hair_01_r [{shapeType=Capsule, length=2, radius=2, shapeName=dyn_rear_hair_01_l_capsule, center=<3.9413 0.05 -0.1689>, rotation=<rot 0 90 -90>}]
 dyn_rear_hair_02_l [{
   shapeType=Capsule
   length=5.4646
   radius=1.778
   shapeName=dyn_rear_hair_02_l_capsule
   center=<-2.374 -0.0753 0.2852>
   rotation=<rot 0.0001 -177.5016 90>
  }]
 dyn_rear_hair_02_r [{
   shapeType=Capsule
   length=5.4646
   radius=1.778
   shapeName=dyn_rear_hair_02_l_capsule
   center=<2.374 0.0753 -0.2852>
   rotation=<rot 0.0001 -177.5016 -90>
  }]
 dyn_rear_hair_03_l [{
   shapeType=Capsule
   length=5.1386
   radius=1.8617
   shapeName=dyn_rear_hair_03_r_capsule
   center=<-0.2349 -0.4048 0.2276>
   rotation=<rot 29.9477 -175.9037 -87.9592>
  }]
 dyn_rear_hair_03_r [{
   shapeType=Capsule
   length=5.1386
   radius=1.8617
   shapeName=dyn_rear_hair_03_r_capsule
   center=<0.2349 0.4048 -0.2276>
   rotation=<rot 29.9477 -175.9037 92.0408>
  }]
 dyn_rear_hairstrand_a_01_l [{
   shapeType=Capsule
   length=3.6275
   radius=1.4601
   shapeName=dyn_rear_hairstrand_a_01_l_capsule
   center=<-3.0056 -0.0952 0.2107>
   rotation=<rot 0 90 90>
  }]
 dyn_rear_hairstrand_a_01_r [{
   shapeType=Capsule
   length=5.0768
   radius=1.5067
   shapeName=dyn_rear_hairstrand_a_01_l_capsule
   center=<3.0056 0.0952 -0.2107>
   rotation=<rot 0 90 -90>
  }]
 dyn_rear_hairstrand_a_02_l [{
   shapeType=Capsule
   length=2.112
   radius=1.888
   shapeName=dyn_rear_hairstrand_a_02_l_capsule
   center=<-2.1692 -0.0857 0.3182>
   rotation=<rot 0 90 90>
  }]
 dyn_rear_hairstrand_a_02_r [{
   shapeType=Capsule
   length=2.112
   radius=1.888
   shapeName=dyn_rear_hairstrand_a_02_l_capsule
   center=<2.1692 0.0857 -0.3182>
   rotation=<rot 0 90 -90>
  }]
 dyn_rear_hairstrand_a_03_l [{
   shapeType=Capsule
   length=3.6922
   radius=1.6108
   shapeName=dyn_rear_hairstrand_a_03_l_capsule
   center=<-2.7413 0.1257 -0.4407>
   rotation=<rot -0.0336 88.1651 89.9683>
  }]
 dyn_rear_hairstrand_a_03_r [{
   shapeType=Capsule
   length=3.6922
   radius=1.6108
   shapeName=dyn_rear_hairstrand_a_03_l_capsule
   center=<2.7413 -0.1257 0.4407>
   rotation=<rot -0.0336 88.1651 -90.0317>
  }]
 dyn_rear_hairstrand_b_01_l [{shapeType=Capsule, length=2, radius=2, shapeName=dyn_rear_hairstrand_b_02_l_capsule, center=<-2.7995 0.119 0.2429>, rotation=<rot 0 90 90>}]
 dyn_rear_hairstrand_b_01_r [{shapeType=Capsule, length=2, radius=2, shapeName=dyn_rear_hairstrand_b_02_l_capsule, center=<2.7995 -0.119 -0.2429>, rotation=<rot 0 90 -90>}]
 dyn_rear_hairstrand_b_02_l [{shapeType=Capsule, length=2, radius=2, shapeName=dyn_rear_hairstrand_c_01_l_capsule, center=<-2.1869 0.0278 0.4108>, rotation=<rot 0 90 90>}]
 dyn_rear_hairstrand_b_02_r [{
   shapeType=Capsule
   length=2
   radius=2
   shapeName=dyn_rear_hairstrand_c_01_l_capsule
   center=<2.1869 -0.0278 -0.4108>
   rotation=<rot 0 90 -90>
  }]
 dyn_rear_hairstrand_c_01_l [{shapeType=Capsule, length=2, radius=2, shapeName=dyn_rear_hairstrand_c_02_l_capsule, center=<-2.8075 0.1636 -0.7771>, rotation=<rot 0 90 90>}]
 dyn_rear_hairstrand_c_01_r [{shapeType=Capsule, length=2, radius=2, shapeName=dyn_rear_hairstrand_c_02_l_capsule, center=<2.8075 -0.1636 0.7771>, rotation=<rot 0 90 -90>}]
 dyn_rear_hairstrand_c_02_l [{
   shapeType=Capsule
   length=2
   radius=2
   shapeName=dyn_rear_hairstrand_c_02_r_capsule
   center=<-1.8783 0.6432 0.7771>
   rotation=<rot 0 80.6345 90>
  }]
 dyn_rear_hairstrand_c_02_r [{
   shapeType=Capsule
   length=2
   radius=2
   shapeName=dyn_rear_hairstrand_c_02_r_capsule
   center=<1.8783 -0.6432 -0.7771>
   rotation=<rot 0 80.6345 -90>
  }]
 dyn_side_hair_01_l [{shapeType=Capsule, length=3.2926, radius=2, shapeName=dyn_side_hair_01_l_capsule, center=<-2.6574 1.5937 0.7657>, rotation=<rot 0 90 90>}]
 dyn_side_hair_01_l_root [{shapeType=Capsule, length=0.505, radius=0.505, shapeName=dyn_side_hair_01_l_root_capsule, rotation=<rot 0 -90 0>}]
 dyn_side_hair_01_r [{shapeType=Capsule, length=3.2926, radius=2, shapeName=dyn_side_hair_01_l_capsule, center=<2.6574 -1.5937 -0.7657>, rotation=<rot 0 90 -90>}]
 dyn_side_hair_01_r_root [{shapeType=Capsule, length=0.505, radius=0.505, shapeName=dyn_side_hair_01_r_root_capsule, rotation=<rot 0 -90 0>}]
 dyn_side_hair_02_l [{shapeType=Capsule, length=2, radius=2, shapeName=dyn_side_hair_02_l_capsule, center=<-2.9905 0.3771 0.6293>, rotation=<rot 0 90 90>}]
 dyn_side_hair_02_r [{shapeType=Capsule, length=2, radius=2, shapeName=dyn_side_hair_02_l_capsule, center=<2.9905 -0.3771 -0.6293>, rotation=<rot 0 90 -90>}]
 dyn_side_hair_03_l [{
   shapeType=Capsule
   length=4.2382
   radius=1.9247
   shapeName=dyn_side_hair_03_l_capsule
   center=<-1.8383 0.0435 0.3308>
   rotation=<rot -0.0001 90 87.4828>
  }]
 dyn_side_hair_03_r [{
   shapeType=Capsule
   length=4.2382
   radius=1.9247
   shapeName=dyn_side_hair_03_l_capsule
   center=<1.8383 -0.0435 -0.3308>
   rotation=<rot -0.0001 90 -92.5172>
  }]
 dyn_side_hair_04_l [{
   shapeType=Capsule
   length=3.4876
   radius=1.6225
   shapeName=dyn_side_hair_04_l_capsule
   center=<-2.9055 0.1038 -0.5132>
   rotation=<rot 0 90 111.6023>
  }]
 dyn_side_hair_04_r [{
   shapeType=Capsule
   length=3.4876
   radius=1.6225
   shapeName=dyn_side_hair_04_l_capsule
   center=<2.9055 -0.1038 0.5132>
   rotation=<rot 0 90 -68.3977>
  }]
 head [
  {shapeType=Capsule, length=7.8644, radius=5.0524, center=<3.4482 0.7204 -0.0102>, rotation=<rot -0.0142 -77.8254 -89.9595>}
  {shapeType=Capsule, length=6.9993, radius=5.0584, shapeName=head_capsule, center=<0.5951 5.2461 0>, rotation=<rot -90 0 0>}
  {shapeType=Capsule, length=15.9371, radius=4.8535, shapeName=head_capsule, center=<6.4345 -2.0724 0>}
 ]
 neck_01 [{shapeType=Capsule, length=11.0079, radius=4.2936, shapeName=neck_01_capsule, rotation=<rot 0 -90.0001 90.093>}]
 spine_05 [{shapeType=Capsule, length=13.5855, radius=9.4784, shapeName=spine_05_capsule, center=<-1.496 -0.8954 0>, rotation=<rot 0 -90 0>}]
 upperarm_l [{
   shapeType=Capsule
   length=22.0676
   radius=4.2053
   shapeName=upperarm_l_capsule
   center=<9.6679 -0.3857 -0.9289>
   rotation=<rot -0.0438 -91.589 88.4293>
  }]
 upperarm_r [{
   shapeType=Capsule
   length=22.0676
   radius=4.2053
   shapeName=upperarm_l_capsule
   center=<-9.6679 0.3857 0.9289>
   rotation=<rot -0.0438 -91.589 -91.5707>
  }]
}
constraints [
 {bone1Name=head, bone2Name=neck_01, swing1LimitDegrees=32, swing2LimitDegrees=32, twistLimitDegrees=32}
 {bone1Name=clavicle_l, bone2Name=spine_05, swing1LimitDegrees=32, swing2LimitDegrees=32, twistLimitDegrees=32}
 {bone1Name=neck_01, bone2Name=spine_05, swing1LimitDegrees=32, swing2LimitDegrees=32, twistLimitDegrees=32}
 {bone1Name=dyn_side_hair_01_l_root, bone2Name=head, swing1LimitDegrees=32, swing2LimitDegrees=32, twistLimitDegrees=32}
 {bone1Name=dyn_side_hair_01_l, bone2Name=dyn_side_hair_01_l_root, swing1LimitDegrees=1, swing2LimitDegrees=10, twistLimitDegrees=10}
 {bone1Name=dyn_side_hair_02_l, bone2Name=dyn_side_hair_01_l, swing1LimitDegrees=1, swing2LimitDegrees=10, twistLimitDegrees=10}
 {bone1Name=dyn_side_hair_03_l, bone2Name=dyn_side_hair_02_l, swing1LimitDegrees=1, swing2LimitDegrees=10, twistLimitDegrees=10}
 {bone1Name=dyn_side_hair_04_l, bone2Name=dyn_side_hair_03_l, swing1LimitDegrees=1, swing2LimitDegrees=10, twistLimitDegrees=10}
 {bone1Name=dyn_side_hair_01_r_root, bone2Name=head, swing1LimitDegrees=32, swing2LimitDegrees=32, twistLimitDegrees=32}
 {bone1Name=dyn_rear_hair_01_r, bone2Name=head, swing1LimitDegrees=4, swing2LimitDegrees=4, twistLimitDegrees=2}
 {bone1Name=dyn_rear_hairstrand_b_01_r, bone2Name=dyn_rear_hair_03_r, swing1LimitDegrees=4, swing2LimitDegrees=12, twistLimitDegrees=4}
 {bone1Name=dyn_rear_hairstrand_c_01_r, bone2Name=dyn_rear_hair_03_r, swing1LimitDegrees=4, swing2LimitDegrees=12, twistLimitDegrees=4}
 {bone1Name=dyn_rear_hairstrand_a_01_r, bone2Name=dyn_rear_hair_03_r, swing1LimitDegrees=4, swing2LimitDegrees=12, twistLimitDegrees=18}
 {bone1Name=dyn_rear_hairstrand_a_02_r, bone2Name=dyn_rear_hairstrand_a_01_r, swing1LimitDegrees=4, swing2LimitDegrees=12, twistLimitDegrees=4}
 {bone1Name=dyn_rear_hair_01_l, bone2Name=head, swing1LimitDegrees=4, swing2LimitDegrees=4, twistLimitDegrees=2}
 {bone1Name=dyn_rear_hair_03_l, bone2Name=dyn_rear_hair_02_l, swing1LimitDegrees=12, swing2LimitDegrees=12, twistLimitDegrees=2}
 {bone1Name=dyn_rear_hairstrand_a_01_l, bone2Name=dyn_rear_hair_03_l, swing1LimitDegrees=4, swing2LimitDegrees=12, twistLimitDegrees=18}
 {bone1Name=dyn_rear_hairstrand_a_02_l, bone2Name=dyn_rear_hairstrand_a_01_l, swing1LimitDegrees=4, swing2LimitDegrees=12, twistLimitDegrees=4}
 {bone1Name=dyn_rear_hairstrand_a_03_l, bone2Name=dyn_rear_hairstrand_a_02_l, swing1LimitDegrees=4, swing2LimitDegrees=12, twistLimitDegrees=4}
 {bone1Name=dyn_rear_hairstrand_b_01_l, bone2Name=dyn_rear_hair_03_l, swing1LimitDegrees=4, swing2LimitDegrees=12, twistLimitDegrees=4}
 {bone1Name=dyn_rear_hairstrand_b_02_l, bone2Name=dyn_rear_hairstrand_b_01_l, swing1LimitDegrees=4, swing2LimitDegrees=12, twistLimitDegrees=4}
 {bone1Name=dyn_rear_hairstrand_c_01_l, bone2Name=dyn_rear_hair_03_l, swing1LimitDegrees=4, swing2LimitDegrees=12, twistLimitDegrees=4}
 {bone1Name=dyn_rear_hairstrand_c_02_l, bone2Name=dyn_rear_hairstrand_c_01_l, swing1LimitDegrees=4, swing2LimitDegrees=12, twistLimitDegrees=4}
 {bone1Name=dyn_rear_hairstrand_a_03_r, bone2Name=dyn_rear_hairstrand_a_02_r, swing1LimitDegrees=4, swing2LimitDegrees=12, twistLimitDegrees=4}
 {bone1Name=dyn_rear_hairstrand_c_02_r, bone2Name=dyn_rear_hairstrand_c_01_r, swing1LimitDegrees=4, swing2LimitDegrees=12, twistLimitDegrees=4}
 {bone1Name=dyn_rear_hairstrand_b_02_r, bone2Name=dyn_rear_hairstrand_b_01_r, swing1LimitDegrees=4, swing2LimitDegrees=12, twistLimitDegrees=4}
 {bone1Name=dyn_rear_hair_02_l, bone2Name=dyn_rear_hair_01_l, swing1LimitDegrees=6, swing2LimitDegrees=6, twistLimitDegrees=2}
 {bone1Name=upperarm_l, bone2Name=clavicle_l, swing1LimitDegrees=32, swing2LimitDegrees=32, twistLimitDegrees=32}
 {bone1Name=dyn_front_hair_01_l_root, bone2Name=head, swing1LimitDegrees=32, swing2LimitDegrees=32, twistLimitDegrees=32}
 {bone1Name=dyn_front_hair_01_l, bone2Name=dyn_front_hair_01_l_root, swing1LimitDegrees=8, swing2LimitDegrees=8, twistLimitDegrees=8}
 {bone1Name=dyn_front_hair_02_l, bone2Name=dyn_front_hair_01_l, swing1LimitDegrees=12, swing2LimitDegrees=12, twistLimitDegrees=12}
 {bone1Name=dyn_front_hair_03_l, bone2Name=dyn_front_hair_02_l, swing1LimitDegrees=32, swing2LimitDegrees=32, twistLimitDegrees=32}
 {bone1Name=dyn_front_hair_04_l, bone2Name=dyn_front_hair_03_l, swing1LimitDegrees=32, swing2LimitDegrees=32, twistLimitDegrees=32}
 {bone1Name=dyn_front_hair_01_r_root, bone2Name=head, swing1LimitDegrees=32, swing2LimitDegrees=32, twistLimitDegrees=32}
 {bone1Name=upperarm_r, bone2Name=clavicle_r, swing1LimitDegrees=32, swing2LimitDegrees=32, twistLimitDegrees=32}
 {bone1Name=dyn_front_hair_04_r, bone2Name=dyn_front_hair_03_r, swing1LimitDegrees=32, swing2LimitDegrees=32, twistLimitDegrees=32}
 {bone1Name=dyn_rear_hair_02_r, bone2Name=dyn_rear_hair_01_r, swing1LimitDegrees=6, swing2LimitDegrees=6, twistLimitDegrees=4}
 {bone1Name=dyn_rear_hair_03_r, bone2Name=dyn_rear_hair_02_r, swing1LimitDegrees=12, swing2LimitDegrees=12, twistLimitDegrees=10}
 {bone1Name=clavicle_r, bone2Name=spine_05, swing1LimitDegrees=32, swing2LimitDegrees=32, twistLimitDegrees=32}
 {bone1Name=dyn_front_hair_03_r, bone2Name=dyn_front_hair_02_r, swing1LimitDegrees=32, swing2LimitDegrees=32, twistLimitDegrees=32}
 {bone1Name=dyn_front_hair_01_r, bone2Name=dyn_front_hair_01_r_root, swing1LimitDegrees=8, swing2LimitDegrees=8, twistLimitDegrees=8}
 {bone1Name=dyn_front_hair_02_r, bone2Name=dyn_front_hair_01_r, swing1LimitDegrees=12, swing2LimitDegrees=12, twistLimitDegrees=12}
 {bone1Name=dyn_side_hair_01_r, bone2Name=dyn_side_hair_01_r_root, swing1LimitDegrees=1, swing2LimitDegrees=10, twistLimitDegrees=3}
 {bone1Name=dyn_side_hair_02_r, bone2Name=dyn_side_hair_01_r, swing1LimitDegrees=1, swing2LimitDegrees=10, twistLimitDegrees=10}
 {bone1Name=dyn_side_hair_03_r, bone2Name=dyn_side_hair_02_r, swing1LimitDegrees=1, swing2LimitDegrees=10, twistLimitDegrees=1}
 {bone1Name=dyn_side_hair_04_r, bone2Name=dyn_side_hair_03_r, swing1LimitDegrees=1, swing2LimitDegrees=10, twistLimitDegrees=1}
]
