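# Animation Blueprint for the Ephemeral character's face accessory part, animating its dynamic hair bones.
# Parent-mesh pose feeds a parachute-aware Control Rig, then two head-space RigidBody solves (standard and backpack physics assets) joined by a bool blend.
# Fixed rear-hair offsets follow, then right-side hair corrections weighted by head rotation relative to spine_05.
# The event graph computes that weight, per-state gravity and a backbling-based ShortenHair flag.
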AnimBlueprint /BRCosmetics/Characters/Player/Female/Medium/Bodies/F_MED_Ephemeral/Meshes/Parts/F_MED_Ephemeral_FaceAcc_AnimBP
target_skeleton=/Game/Characters/Player/Female/Medium/Bodies/F_MED_Ephemeral/Meshes/Parts/F_MED_Ephemeral_FaceAcc_Skeleton
graphs [
 AnimGraph {nodes [
   n1 = Get Head ADS Up Alpha() -> (Head ADS Up Alpha:double)
   n2 = "Transform (Modify) Bone - Bone: dyn_side_hair_02_r"(
    ComponentPose=Local To Component(LocalPose=Use cached pose 'PostBackpackLogic'())
    Alpha=@n1
    Translation="-0.034,-0.757,1.8"
    Rotation="-30,5.228,-40"
    Scale="0.9,0.9,0.9"
   ) {
    rotationMode=BMM_Additive
    scaleMode=BMM_Additive
    translationMode=BMM_Additive
    alphaBoolBlend {blendOption=Linear}
    boneToModify {boneName=dyn_side_hair_02_r}
    rotation=<rot -30 5.228 -40>
    scale=<0.9 0.9 0.9>
    translation=<-0.034 -0.757 1.8>
   }
   n3 = "Transform (Modify) Bone - Bone: dyn_side_hair_03_r"(ComponentPose=@n2, Alpha=@n1, Translation="1,0,0", Rotation="0,-0.566,-3.375", Scale="0.8,0.8,0.8") {
    rotationMode=BMM_Additive
    scaleMode=BMM_Additive
    translationMode=BMM_Additive
    alphaBoolBlend {blendOption=Linear}
    boneToModify {boneName=dyn_side_hair_03_r}
    rotation {roll=-3.375, yaw=-0.566}
    scale=<0.8 0.8 0.8>
    translation {x=1}
   }
   n4 = "Transform (Modify) Bone - Bone: dyn_rear_hair_02_r"(ComponentPose=@n3, Alpha=@n1, Translation="0,0,0", Rotation="0,0,0", Scale="0.8,0.8,0.8") {
    rotationMode=BMM_Additive
    scaleMode=BMM_Additive
    translationMode=BMM_Additive
    alphaBoolBlend {blendOption=Linear}
    boneToModify {boneName=dyn_rear_hair_02_r}
    scale=<0.8 0.8 0.8>
   }
   Output Pose(Result=Component To Local(ComponentPose=@n4)) {layerGroup=None}
   n5 = Control Rig!(Source=Copy Pose From Mesh() {bUseAttachedParent=true}, Alpha, IsParachuting:bool=Get bIsParachuteOpened()) -> (Pose) {
    alphaBoolBlend {blendOption=Linear}
    controlRigAssetReference {blueprintRigClass=&/F_MED_Ephemeral_FaceAcc_CtrlRig.F_MED_Ephemeral_FaceAcc_CtrlRig_C}
   }
   Save cached pose 'PostControlRig'(Pose=@n5) {cacheName=PostControlRig}
   n6 = "Transform (Modify) Bone - Bone: dyn_rear_hair_01_l"(ComponentPose=Local To Component(LocalPose=Use cached pose 'PostRBAN'()), Alpha=1, Translation="0,0,2", Rotation="0,0,0", Scale="1,1,1") {
    scaleMode=BMM_Replace
    translationMode=BMM_Additive
    alphaBoolBlend {blendOption=Linear}
    boneToModify {boneName=dyn_rear_hair_01_l}
    translation {z=2}
   }
   n7 = "Transform (Modify) Bone - Bone: dyn_rear_hair_02_l"(ComponentPose=@n6, Alpha=1, Translation="0,0,2", Rotation="0,0,0", Scale="1,1,1") {
    scaleMode=BMM_Replace
    translationMode=BMM_Additive
    alphaBoolBlend {blendOption=Linear}
    boneToModify {boneName=dyn_rear_hair_02_l}
    translation {z=2}
   }
   n8 = "Transform (Modify) Bone - Bone: dyn_rear_hairstrand_a_02_l"(ComponentPose=@n7, Alpha=1, Translation="0,0,0", Rotation="0,0,20", Scale="1,1,1") {rotationMode=BMM_Additive, alphaBoolBlend {blendOption=Linear}, boneToModify {boneName=dyn_rear_hairstrand_a_02_l}, rotation {roll=20}}
   n9 = "Transform (Modify) Bone - Bone: dyn_rear_hairstrand_b_02_l"(ComponentPose=@n8, Alpha=1, Translation="0,0,0", Rotation="0,0,20", Scale="1,1,1") {rotationMode=BMM_Additive, alphaBoolBlend {blendOption=Linear}, boneToModify {boneName=dyn_rear_hairstrand_b_02_l}, rotation {roll=20}}
   n10 = "Transform (Modify) Bone - Bone: dyn_rear_hairstrand_c_02_l"(ComponentPose=@n9, Alpha=1, Translation="0,0,0", Rotation="0,0,15", Scale="1,1,1") {rotationMode=BMM_Additive, alphaBoolBlend {blendOption=Linear}, boneToModify {boneName=dyn_rear_hairstrand_c_02_l}, rotation {roll=15}}
   n11 = "Transform (Modify) Bone - Bone: dyn_rear_hair_01_r"(ComponentPose=@n10, Alpha=1, Translation="0,0,2", Rotation="0,0,0", Scale="1,1,1") {
    scaleMode=BMM_Replace
    translationMode=BMM_Additive
    alphaBoolBlend {blendOption=Linear}
    boneToModify {boneName=dyn_rear_hair_01_r}
    translation {z=2}
   }
   n12 = "Transform (Modify) Bone - Bone: dyn_rear_hair_02_r"(ComponentPose=@n11, Alpha=1, Translation="0,0,2", Rotation="0,0,0", Scale="1,1,1") {
    scaleMode=BMM_Replace
    translationMode=BMM_Additive
    alphaBoolBlend {blendOption=Linear}
    boneToModify {boneName=dyn_rear_hair_02_r}
    translation {z=2}
   }
   n13 = "Transform (Modify) Bone - Bone: dyn_rear_hairstrand_a_02_r"(ComponentPose=@n12, Alpha=1, Translation="0,0,0", Rotation="0,0,20", Scale="1,1,1") {rotationMode=BMM_Additive, alphaBoolBlend {blendOption=Linear}, boneToModify {boneName=dyn_rear_hairstrand_a_02_r}, rotation {roll=20}}
   n14 = "Transform (Modify) Bone - Bone: dyn_rear_hairstrand_b_02_r"(ComponentPose=@n13, Alpha=1, Translation="0,0,0", Rotation="0,0,20", Scale="1,1,1") {rotationMode=BMM_Additive, alphaBoolBlend {blendOption=Linear}, boneToModify {boneName=dyn_rear_hairstrand_b_02_r}, rotation {roll=20}}
   n15 = "Transform (Modify) Bone - Bone: dyn_rear_hairstrand_c_02_r"(ComponentPose=@n14, Alpha=1, Translation="0,0,0", Rotation="0,0,15", Scale="1,1,1") {rotationMode=BMM_Additive, alphaBoolBlend {blendOption=Linear}, boneToModify {boneName=dyn_rear_hairstrand_c_02_r}, rotation {roll=15}}
   Save cached pose 'PostBackpackLogic'(Pose=Component To Local(ComponentPose=@n15)) {cacheName=PostBackpackLogic}
   n16 = RigidBody!(
    ComponentPose=Local To Component(LocalPose=Use cached pose 'PostControlRig'())
    Alpha=0.9
    OverrideWorldGravity:Vector="0,0,-980"
    ExternalForce
   ) -> (Pose) {
    bOverrideWorldGravity=true
    overridePhysicsAsset=&/F_MED_Ephemeral_FaceAcc_Rear_Physics
    simulationSpace=BaseBoneSpace
    alphaBoolBlend {blendOption=Linear}
    baseBoneRef {boneName=head}
    componentAppliedLinearAccClamp=<1400 1400 1400>
    componentLinearAccScale=<0.3 0.3 0.3>
    componentLinearVelScale=<0.3 0.3 0.3>
    overrideWorldGravity {z=-980}
    simSpaceSettings {maxAngularAcceleration=50, maxAngularVelocity=50, maxLinearAcceleration=50, maxLinearVelocity=50, worldAlpha=1}
   }
   n17 = RigidBody(ComponentPose=Local To Component(LocalPose=Use cached pose 'PostControlRig'()), Alpha=0.9, OverrideWorldGravity="0,0,-980") {
    bOverrideWorldGravity=true
    overridePhysicsAsset=&/F_MED_Ephemeral_FaceAcc_Rear_Backpack_Physics
    simulationSpace=BaseBoneSpace
    alphaBoolBlend {blendOption=Linear}
    baseBoneRef {boneName=head}
    componentAppliedLinearAccClamp=<1400 1400 1400>
    componentLinearAccScale=<0.3 0.3 0.3>
    componentLinearVelScale=<0.3 0.3 0.3>
    overrideWorldGravity {z=-980}
    simSpaceSettings {maxAngularAcceleration=50, maxAngularVelocity=50, maxLinearAcceleration=50, maxLinearVelocity=50, worldAlpha=1}
   }
   Save cached pose 'PostRBAN'(Pose=Blend Poses by bool(BlendPose_0=Component To Local(ComponentPose=@n17), BlendPose_1=Component To Local(ComponentPose=@n16)) {blendType=Linear}) {cacheName=PostRBAN}
  ]}
 EventGraph {nodes [
   Event BlueprintUpdateAnimation()
   > n1 = Sequence!(execute) -> (then_0, then_1, then_2:Exec)
   n2 = GetOwnerSkeletalMeshComponent()
   n3 = BreakRotator(InRot=BreakTransform(InTransform=GetDeltaTransformFromRefPose(self=@n2, BoneName=head, BaseName=spine_05)).Rotation)
   n4 = MapRangeClamped(Value=@n3, InRangeA=0, InRangeB=15)
   n5 = MapRangeClamped(Value=@n3.Yaw, InRangeA=-40, InRangeB=0, OutRangeA=1, OutRangeB=0)
   n6 = MapRangeClamped(Value=@n3.Pitch, InRangeA=0, InRangeB=50)
   Set Head ADS Up Alpha(
    execute=@n1.then_1
    Head ADS Up Alpha:double=float * float(A=float + float!(A=@n4, B=@n6, C:double=@n5) -> (ReturnValue), B=0.33) {is=title K2Node_PromotableOperator}
   ) -> (Output_Get:double)
   n7 = Gravity Override State Function DEPRECATED!(
    execute=@n1.then_0
    CharacterPartAnimInstance=Self-Reference()
    GameTimeInSeconds=GetGameTimeInSeconds()
    GravityOverrideParams_Skydive_GoingUp_28_#1:Gravity Override Setting Struct Structure=(WindFrequency_28_#2=(X=1.5,Y=7,Z=8),WindAmplitude_31_#3=(X=200,Y=-500,Z=400),WindOffset_32_#4=(X=0,Y=0,Z=0),GravityOverride_15_#5=(X=0,Y=-300,Z=1500),JointName_27_#6="root")
    GravityOverrideParams_Skydive_FloatingDown_29_#7:Gravity Override Setting Struct Structure=(WindFrequency_28_#2=(X=1.5,Y=7,Z=8),WindAmplitude_31_#3=(X=200,Y=-500,Z=400),WindOffset_32_#4=(X=0,Y=-600,Z=400),GravityOverride_15_#5=(X=0,Y=-700,Z=600),JointName_27_#6="root")
    GravityOverrideParams_Skydive_DivingDown_30_#8:Gravity Override Setting Struct Structure=(WindFrequency_28_#2=(X=1.5,Y=12,Z=10),WindAmplitude_31_#3=(X=50,Y=300,Z=100),WindOffset_32_#4=(X=0,Y=0,Z=0),GravityOverride_15_#5=(X=-1500,Y=0,Z=0),JointName_27_#6="pelvis")
    GravityOverrideParams_Skydive_Parachuting_31_#9:Gravity Override Setting Struct Structure=(WindFrequency_28_#2=(X=1.5,Y=7,Z=8),WindAmplitude_31_#3=(X=200,Y=-500,Z=400),WindOffset_32_#4=(X=0,Y=0,Z=0),GravityOverride_15_#5=(X=0,Y=-800,Z=-980),JointName_27_#6="root")
    GravityOverrideParams_OnGround_Standing_32_#10:Gravity Override Setting Struct Structure=(WindFrequency_28_#2=(X=0,Y=0,Z=0),WindAmplitude_31_#3=(X=0,Y=0,Z=0),WindOffset_32_#4=(X=0,Y=0,Z=0),GravityOverride_15_#5=(X=0,Y=0,Z=-980),JointName_27_#6="root")
    GravityOverrideParams_OnGround_Moving_33_#11:Gravity Override Setting Struct Structure=(WindFrequency_28_#2=(X=0,Y=0,Z=0),WindAmplitude_31_#3=(X=0,Y=0,Z=0),WindOffset_32_#4=(X=0,Y=0,Z=0),GravityOverride_15_#5=(X=0,Y=0,Z=-980),JointName_27_#6="root")
    GravityOverrideParams_EmoteOrMelee_34_#12:Gravity Override Setting Struct Structure=(WindFrequency_28_#2=(X=0,Y=0,Z=0),WindAmplitude_31_#3=(X=0,Y=0,Z=0),WindOffset_32_#4=(X=0,Y=0,Z=0),GravityOverride_15_#5=(X=0,Y=0,Z=-980),JointName_27_#6="root")
    GravityOverrideParams_Swimming_35_#13:Gravity Override Setting Struct Structure=(WindFrequency_28_#2=(X=0,Y=0,Z=0),WindAmplitude_31_#3=(X=0,Y=0,Z=0),WindOffset_32_#4=(X=0,Y=0,Z=0),GravityOverride_15_#5=(X=-800,Y=0,Z=0),JointName_27_#6="pelvis")
    GravityOverrideParams_Front_End_36_#14:Gravity Override Setting Struct Structure=(WindFrequency_28_#2=(X=0,Y=0,Z=0),WindAmplitude_31_#3=(X=0,Y=0,Z=0),WindOffset_32_#4=(X=0,Y=0,Z=0),GravityOverride_15_#5=(X=0,Y=0,Z=-980),JointName_27_#6="root")
    GravityOverrideParams_DBNO_40_#15:Gravity Override Setting Struct Structure=(WindFrequency_28_#2=(X=0,Y=0,Z=0),WindAmplitude_31_#3=(X=0,Y=0,Z=0),WindOffset_32_#4=(X=0,Y=0,Z=0),GravityOverride_15_#5=(X=-980,Y=400,Z=0),JointName_27_#6="pelvis")
    GravityOverrideParams_DBNO_Carried_41_#16:Gravity Override Setting Struct Structure=(WindFrequency_28_#2=(X=0,Y=0,Z=0),WindAmplitude_31_#3=(X=0,Y=0,Z=0),WindOffset_32_#4=(X=0,Y=0,Z=0),GravityOverride_15_#5=(X=-980,Y=500,Z=-500),JointName_27_#6="pelvis")
    GravityOverrideParams_OnGround_Falling_45_#17:Gravity Override Setting Struct Structure=(WindFrequency_28_#2=(X=0,Y=0,Z=0),WindAmplitude_31_#3=(X=0,Y=0,Z=0),WindOffset_32_#4=(X=0,Y=0,Z=0),GravityOverride_15_#5=(X=0,Y=0,Z=-980),JointName_27_#6="root")
    GravityOverrideParams_Ziplining_46_#18:Gravity Override Setting Struct Structure=(WindFrequency_28_#2=(X=0,Y=0,Z=0),WindAmplitude_31_#3=(X=0,Y=0,Z=0),WindOffset_32_#4=(X=0,Y=0,Z=0),GravityOverride_15_#5=(X=0,Y=0,Z=-980),JointName_27_#6="root")
    GravityOverrideParams_Skydive_Surfing_49_#19:Gravity Override Setting Struct Structure=(WindFrequency_28_#2=(X=1.5,Y=7,Z=8),WindAmplitude_31_#3=(X=200,Y=-500,Z=400),WindOffset_32_#4=(X=0,Y=0,Z=0),GravityOverride_15_#5=(X=500,Y=-300,Z=-980),JointName_27_#6="root")
    GravityOverrideParams_Riding_Standing_54_#20:Gravity Override Setting Struct Structure=(WindFrequency_28_#2=(X=0,Y=0,Z=0),WindAmplitude_31_#3=(X=0,Y=0,Z=0),WindOffset_32_#4=(X=0,Y=0,Z=0),GravityOverride_15_#5=(X=0,Y=-350,Z=-980),JointName_27_#6="root")
    GravityOverrideParams_Riding_Moving_55_#21:Gravity Override Setting Struct Structure=(WindFrequency_28_#2=(X=0,Y=0,Z=0),WindAmplitude_31_#3=(X=0,Y=0,Z=0),WindOffset_32_#4=(X=0,Y=0,Z=0),GravityOverride_15_#5=(X=0,Y=-450,Z=-450),JointName_27_#6="root")
    GravityOverrideParams_Riding_Moving_Targeting_58_#22:Gravity Override Setting Struct Structure=(WindFrequency_28_#2=(X=0,Y=0,Z=0),WindAmplitude_31_#3=(X=0,Y=0,Z=0),WindOffset_32_#4=(X=0,Y=0,Z=0),GravityOverride_15_#5=(X=0,Y=-200,Z=-980),JointName_27_#6="root")
    GravityOverrideParams_Tactical_Sprint_57_#23:Gravity Override Setting Struct Structure=(WindFrequency_28_#2=(X=0,Y=0,Z=0),WindAmplitude_31_#3=(X=0,Y=0,Z=0),WindOffset_32_#4=(X=0,Y=0,Z=0),GravityOverride_15_#5=(X=0,Y=-450,Z=-500),JointName_27_#6="root")
    GravityOverrideParams_Rail_Grinding_61_#24:Gravity Override Setting Struct Structure=(WindFrequency_28_#2=(X=0,Y=0,Z=0),WindAmplitude_31_#3=(X=0,Y=0,Z=0),WindOffset_32_#4=(X=0,Y=0,Z=0),GravityOverride_15_#5=(X=0,Y=0,Z=-980),JointName_27_#6="root")
    GravityOverrideParams_DBNO_Tumbling_67_#25:Gravity Override Setting Struct Structure=(WindFrequency_28_#2=(X=0,Y=0,Z=0),WindAmplitude_31_#3=(X=0,Y=0,Z=0),WindOffset_32_#4=(X=0,Y=0,Z=0),GravityOverride_15_#5=(X=-2000,Y=0,Z=0),JointName_27_#6="pelvis")
    GravityOverrideParams_DBNO_RollFwd_68_#26:Gravity Override Setting Struct Structure=(WindFrequency_28_#2=(X=0,Y=0,Z=0),WindAmplitude_31_#3=(X=0,Y=0,Z=0),WindOffset_32_#4=(X=0,Y=0,Z=0),GravityOverride_15_#5=(X=-2000,Y=0,Z=0),JointName_27_#6="pelvis")
    GravityOverrideParams_DBNO_RollBwd_69_#27:Gravity Override Setting Struct Structure=(WindFrequency_28_#2=(X=0,Y=0,Z=0),WindAmplitude_31_#3=(X=0,Y=0,Z=0),WindOffset_32_#4=(X=0,Y=0,Z=0),GravityOverride_15_#5=(X=-4000,Y=0,Z=0),JointName_27_#6="pelvis")
    GravityOverrideParams_DBNO_Impulse_70_#28:Gravity Override Setting Struct Structure=(WindFrequency_28_#2=(X=0,Y=0,Z=0),WindAmplitude_31_#3=(X=0,Y=0,Z=0),WindOffset_32_#4=(X=0,Y=0,Z=0),GravityOverride_15_#5=(X=-2000,Y=0,Z=0),JointName_27_#6="pelvis")
    RBAN Params
    DebugDraw
    DebugJointName
   ) -> (then, GravityOverride, CurrentStateOut, GravityOverrideSettingsOut, RBAN SettingsOut)
   > Set Gravity Override(Gravity Override=@n7)
   n8 = GetOwnerSkeletalMeshComponent()
   n9 = BreakRotator(InRot=BreakTransform(InTransform=GetDeltaTransformFromRefPose(self=@n8, BoneName=head, BaseName=spine_05)).Rotation)
   Set Alpha Output(
    execute=@n1.then_2
    Alpha Output:float=float * float(A=MapRangeClamped(Value=@n9.Yaw, InRangeA=-45, InRangeB=0, OutRangeA=1, OutRangeB=0.5), B:bool=Get ShortenHair() -> (ShortenHair:bool))
   ) -> (Output_Get:float)
   Event OnCharacterCustomizationLoaded()
   n10 = CheckForBackbling()
   > Set ShortenHair(ShortenHair:bool=OR Boolean(A=@n10, B=@n10.bIsBackblingACape)) -> (Output_Get:bool)
  ]}
]
variables [
 Alpha Output {type=number}
 Gravity Override {type=Vector}
 Head ADS Up Alpha {type=number}
 Is Backbling A Cape {type=boolean}
 Output {type=CustomCharacterPart, default=null}
 ShortenHair {type=boolean}
]
